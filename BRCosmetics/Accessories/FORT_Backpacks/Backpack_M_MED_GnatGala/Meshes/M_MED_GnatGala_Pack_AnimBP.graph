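# Animation Blueprint for the GnatGala backpack, driving its dangling ropes and the visibility of two pack elements.
# Pack-bone offsets chosen by gender and skydiving state feed Anim Dynamics rope chains, with separate tuning for front end and gameplay.
# Update events set skydive-dependent gravity and inertia variables, though the rope nodes show literal values.
# Material parameters hide the element matching an equipped GnatGala Axe or Machete, restoring it during skydiving or non-sliding tactical sprint.
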AnimBlueprint /BRCosmetics/Accessories/FORT_Backpacks/Backpack_M_MED_GnatGala/Meshes/M_MED_GnatGala_Pack_AnimBP
target_skeleton=&/M_MED_GnatGala_Pack_Skeleton
graphs [
 AnimGraph {nodes [
   n1 = Anim Dynamics - Bone: dyn_upper_rope_01_R!(
    ComponentPose=Local To Component(LocalPose=Use cached pose 'main cache'())
    Alpha=1
    GravityOverride:Vector="0,0,-980"
    ComponentLinearAccScale:Vector="0.5,0.5,0.5"
    ComponentLinearVelScale:Vector="0.5,0.5,0.5"
   ) -> (Pose) {
    angularDampingOverride=0.9
    angularSpringConstant=50
    bUseGravityOverride=true
    bUsePlanarLimit=false
    simulationSpace=RootRelative
    alphaBoolBlend {blendOption=Linear}
    boundBone {boneName=dyn_upper_rope_01_R}
    componentAppliedLinearAccClamp=<1000 1000 1000>
    componentLinearAccScale=<0.5 0.5 0.5>
    componentLinearVelScale=<0.5 0.5 0.5>
    physicsBodyDefinitions [{
      boundBone {boneName=dyn_upper_rope_01_R}
      boxExtents=<10 17 1.5>
      constraintSetup {angularLimitsMax=<0 75 0>, angularLimitsMin=<0 0 0>, angularTarget=<1 0 0>}
      localJointOffset=<2 0 3>
     }]
   }
   n2 = Anim Dynamics - Bone: dyn_lower_rope_01_L!(
    ComponentPose=@n1
    Alpha=1
    GravityOverride:Vector="0,0,-980"
    ComponentLinearAccScale:Vector="0.5,0.5,0.5"
    ComponentLinearVelScale:Vector="0.5,0.5,0.5"
   ) -> (Pose) {
    angularDampingOverride=0.97
    angularSpringConstant=10000
    bOverrideAngularDamping=true
    bUseGravityOverride=true
    bUsePlanarLimit=false
    simulationSpace=RootRelative
    alphaBoolBlend {blendOption=Linear}
    boundBone {boneName=dyn_lower_rope_01_L}
    componentAppliedLinearAccClamp=<250 250 250>
    componentLinearAccScale=<0.5 0.5 0.5>
    componentLinearVelScale=<0.5 0.5 0.5>
    physicsBodyDefinitions [{
      sphereCollisionRadius=1
      boundBone {boneName=dyn_lower_rope_01_L}
      boxExtents=<17 23 1.5>
      constraintSetup {angularLimitsMax=<0 35 0>, angularLimitsMin=<0 -55 0>, angularTarget=<1 0 0>}
      localJointOffset=<3 0 0>
     }]
   }
   n3 = Anim Dynamics - Bone: dyn_lower_rope_01_R!(
    ComponentPose=@n2
    Alpha=1
    GravityOverride:Vector="0,0,-980"
    ComponentLinearAccScale:Vector="0.5,0.5,0.5"
    ComponentLinearVelScale:Vector="0.5,0.5,0.5"
   ) -> (Pose) {
    angularDampingOverride=0.99
    angularSpringConstant=10000
    bOverrideAngularDamping=true
    bUseGravityOverride=true
    bUsePlanarLimit=false
    simulationSpace=RootRelative
    alphaBoolBlend {blendOption=Linear}
    boundBone {boneName=dyn_lower_rope_01_R}
    componentAppliedLinearAccClamp=<250 250 250>
    componentLinearAccScale=<0.5 0.5 0.5>
    componentLinearVelScale=<0.5 0.5 0.5>
    physicsBodyDefinitions [{
      sphereCollisionRadius=1
      boundBone {boneName=dyn_lower_rope_01_R}
      boxExtents=<15 20 1.5>
      constraintSetup {angularLimitsMax=<0 35 0>, angularLimitsMin=<0 -55 0>, angularTarget=<1 0 0>}
      localJointOffset=<3 0 0>
     }]
   }
   n4 = Anim Dynamics - Bone: dyn_upper_rope_01_R(
    ComponentPose=Local To Component(LocalPose=Use cached pose 'main cache'())
    Alpha=1
    GravityOverride="0,0,-980"
    ComponentLinearAccScale="0.9,0.9,0.9"
    ComponentLinearVelScale="0.9,0.9,0.9"
   ) {
    angularDampingOverride=0.9
    angularSpringConstant=50
    bUseGravityOverride=true
    bUsePlanarLimit=false
    simulationSpace=RootRelative
    alphaBoolBlend {blendOption=Linear}
    boundBone {boneName=dyn_upper_rope_01_R}
    componentAppliedLinearAccClamp=<1000 1000 1000>
    componentLinearAccScale=<0.5 0.5 0.5>
    componentLinearVelScale=<0.5 0.5 0.5>
    physicsBodyDefinitions [{
      boundBone {boneName=dyn_upper_rope_01_R}
      boxExtents=<13 20 5>
      constraintSetup {angularLimitsMax=<0 75 0>, angularLimitsMin=<0 0 0>, angularTarget=<1 0 0>}
      localJointOffset=<2 0 2.5>
     }]
   }
   n5 = Anim Dynamics - Bone: dyn_lower_rope_01_L(ComponentPose=@n4, Alpha=1, GravityOverride="0,0,-980", ComponentLinearAccScale="0.9,0.9,0.9", ComponentLinearVelScale="0.9,0.9,0.9") {
    angularDampingOverride=0.97
    angularSpringConstant=50
    bOverrideAngularDamping=true
    bUseGravityOverride=true
    bUsePlanarLimit=false
    simulationSpace=RootRelative
    alphaBoolBlend {blendOption=Linear}
    boundBone {boneName=dyn_lower_rope_01_L}
    componentAppliedLinearAccClamp=<1000 1000 1000>
    componentLinearAccScale=<0.5 0.5 0.5>
    componentLinearVelScale=<0.5 0.5 0.5>
    physicsBodyDefinitions [{
      sphereCollisionRadius=1
      boundBone {boneName=dyn_lower_rope_01_L}
      boxExtents=<27 32 5>
      constraintSetup {angularLimitsMax=<0 35 0>, angularLimitsMin=<0 -55 0>, angularTarget=<1 0 0>}
      localJointOffset=<3 0 0>
     }]
   }
   n6 = Anim Dynamics - Bone: dyn_lower_rope_01_R(ComponentPose=@n5, Alpha=1, GravityOverride="0,0,-980", ComponentLinearAccScale="0.9,0.9,0.9", ComponentLinearVelScale="0.9,0.9,0.9") {
    angularDampingOverride=0.99
    angularSpringConstant=500
    bAngularSpring=true
    bOverrideAngularDamping=true
    bUseGravityOverride=true
    bUsePlanarLimit=false
    simulationSpace=RootRelative
    alphaBoolBlend {blendOption=Linear}
    boundBone {boneName=dyn_lower_rope_01_R}
    componentAppliedLinearAccClamp=<1000 1000 1000>
    componentLinearAccScale=<0.5 0.5 0.5>
    componentLinearVelScale=<0.5 0.5 0.5>
    physicsBodyDefinitions [{
      sphereCollisionRadius=1
      boundBone {boneName=dyn_lower_rope_01_R}
      boxExtents=<22 27 5>
      constraintSetup {angularLimitsMax=<0 35 0>, angularLimitsMin=<0 -55 0>, angularTarget=<1 0 0>}
      localJointOffset=<3 0 0>
     }]
   }
   Output Pose(Result=Blend Poses by bool(BlendPose_0=Component To Local(ComponentPose=@n6), BlendPose_1=Component To Local(ComponentPose=@n3), bActiveValue=Get bIsInFrontEnd()) {blendType=Linear}) {layerGroup=None}
   Save cached pose 'start pose'(Pose=Input Pose()) {cacheName=start pose}
   n7 = "Transform (Modify) Bone - Bone: pack"!(ComponentPose=Local To Component(LocalPose=Use cached pose 'start pose'()), Alpha, Translation="0,0,0", Rotation="0,0,20") -> (Pose) {
    rotationMode=BMM_Additive
    rotationSpace=BCS_BoneSpace
    translationMode=BMM_Additive
    translationSpace=BCS_BoneSpace
    alphaBoolBlend {blendOption=Linear}
    boneToModify {boneName=pack}
    rotation {roll=20}
   }
   n8 = Component To Local(ComponentPose=@n7)
   n9 = Use cached pose 'start pose'()
   n10 = Get isSkyDiving() -> (isSkyDiving:bool)
   n11 = "Transform (Modify) Bone - Bone: pack"!(ComponentPose=Local To Component(LocalPose=Use cached pose 'start pose'()), Alpha, Translation="0,1.5,0") -> (Pose) {
    translationMode=BMM_Additive
    translationSpace=BCS_BoneSpace
    alphaBoolBlend {blendOption=Linear}
    boneToModify {boneName=pack}
    translation {y=1.5}
   }
   Save cached pose 'main cache'(Pose=Blend Poses by bool(
     BlendPose_0=Blend Poses by bool(BlendPose_0=@n8, BlendPose_1=@n9, BlendTime_0=0.5, BlendTime_1=0.5, bActiveValue=@n10) {blendType=Linear}
     BlendPose_1=Component To Local(ComponentPose=@n11)
     bActiveValue=Get isMale() -> (isMale:bool)
    ) {blendType=Linear}) {cacheName=main cache}
  ]}
 EventGraph {nodes [
   Event BlueprintUpdateAnimation()
   > Set isSkyDiving(isSkyDiving:bool=OR Boolean(A=Get bIsSkydiveDiveMode(), B=Get bIsSkydiveFloating())) -> (Output_Get:bool)
   n1 = Get isSkyDiving()
   > Set Gravity Override(Gravity Override=SelectVector(A="0,0,100", B="0,0,-980", bPickA=@n1))
   > Set Component Linear Acc Scale(Component Linear Acc Scale:Vector=SelectVector(A="0.1,0.1,0.1", B="0.5,0.5,0.5", bPickA=@n1)) -> (Output_Get:Vector)
   > Set Component Linear Vel Scale(Component Linear Vel Scale:Vector=SelectVector(A="0.1,0.1,0.1", B="0.5,0.5,5", bPickA=@n1)) -> (Output_Get:Vector)
   n2 = Event OnBeginSliding() {is=title K2Node_Event}
   n3 = PickaxeItemDefinitionHasAnyMetaTag(Tags=Make Array("[0]"=(TagName="Cosmetics.Metatags.GnatGala.Axe")), bRequiresEquipped=true)
   n4 = PickaxeItemDefinitionHasAnyMetaTag(Tags=Make Array("[0]"=(TagName="Cosmetics.Metatags.GnatGala.Machete")), bRequiresEquipped=true)
   n5 = OR Boolean(A=@n3, B=@n4)
   n6 = Branch(execute=[@n11.then, @n2.then, @n22.then], Condition=@n5)
   n7 = NOT Boolean(A=Get bIsSliding())
   n8 = Get bIsTacticalSprinting()
   > n9 = Branch(Condition=OR Boolean(A=Get bIsSkyDiving(), B=AND Boolean(A=@n8, B=@n7)))
   n10 = Sequence(execute=[@n9.then, @n6.else, @n20.then, @n21.then, @n18.else])
   n11 = Get Dynamic Material Instance(execute=[@n16.then, @n14.then_1], SkeletalMeshComponent=GetPartSkeletalMeshComponent())
   SetScalarParameterValue(execute=@n10.then_0, self=@n11, ParameterName=Hide Element 01, Value=0)
   SetScalarParameterValue(execute=@n10.then_1, self=@n11, ParameterName=Hide Element 02, Value=0)
   n12 = Sequence(execute=[@n9.else, @n18.then])
   > Branch(Condition=@n3)
   > SetScalarParameterValue(self=@n11, ParameterName=Hide Element 01, Value=1)
   Branch(execute=@n12.then_1, Condition=@n4)
   > SetScalarParameterValue(self=@n11, ParameterName=Hide Element 02, Value=1)
   n13 = Event OnCharacterCustomizationLoaded()
   > n14 = Sequence()
   n15 = Get Array Item with EFort Custom Part Type(Array:Array of Custom Character Part Object References=@n13.CustomParts) -> (Output:Custom Character Part Object Reference)
   > Set isMale(isMale:bool="Equal (Enum)"(
     A:EFortCustomGender Enum=Get GenderPermitted!(self:Custom Character Part Object Reference=@n15) -> (GenderPermitted:EFortCustomGender Enum=Invalid)
     B:EFortCustomGender Enum=Male
    )) -> (Output_Get:bool)
   n16 = Event OnWeaponChanged()
   n17 = Event OnEndTacticalSprinting() {is=title K2Node_Event}
   n18 = Branch(execute=[@n17.then, @n19.then], Condition=@n5)
   n19 = Event OnEndSkydiving()
   n20 = Event OnBeginTacticalSprinting() {is=title K2Node_Event}
   n21 = Event OnBeginSkydiving()
   n22 = Event OnEndSliding() {is=title K2Node_Event}
   TryGetPawnOwner()
  ]}
]
variables [
 Collection {type=MaterialParameterCollection, default=null}
 Component Linear Acc Scale {type=Vector, default=<0.5 0.5 0.5>}
 Component Linear Vel Scale {type=Vector, default=<0.5 0.5 0.5>}
 Gravity Override {type=Vector}
 isMale {type=boolean}
 isSkyDiving {type=boolean}
]
